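# Niagara impact effect for the GiraffeScallion pickaxe, a toon-styled burst of hearts, star, ring, flare, smoke, speed lines and spikes.
# Seven burst-spawned emitters mix mesh and sprite renderers, with short curve-driven scale, color and dynamic-material animation, mostly multiplied by a system GlobalScale.
# The star adds a time-of-day-tinted flash flare sprite with separate front-end scaling.
# Hearts, star and ring set a Fresnel thickness exponent per platform, with a separate value for targeted profiles including mobile.
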NiagaraSystem /BRCosmetics/Weapons/FORT_Melee/Pickaxe_GiraffeScallion/FX/NS_Pickaxe_GiraffeScallion_Impact
dependencies {
 usedDataInterfaces [
  &NiagaraDataInterfaceCurve
  &NiagaraDataInterfaceVector2DCurve
  &NiagaraDataInterfacePlatformSet
  &NiagaraDataInterfaceVectorCurve
  &NiagaraDataInterfaceColorCurve
 ]
 usedDynamicInputs []
 usedModules [
  &self:SystemSpawnScript.SystemScriptSource.SystemScriptGraph.NiagaraNodeAssignment_0.SetVariables_#1
  &self:SystemSpawnScript.SystemScriptSource.SystemScriptGraph.NiagaraNodeAssignment_1.SetVariables_#2
  &SystemState
 ]
}
emitters {
 Toon_Hearts {
  emitterUpdateScript [
   &EmitterState()
   &SpawnBurst_Instantaneous(
    Spawn Count=&RandomRangeInt(Minimum=3, Maximum=6)
    Use Spawn Probability=false
    UseLoopCountLimit=false
   )
   /SetVariables(Emitter.FresnelThickEXP:NiagaraFloat=&self:PlatFromSelect(
     Default:NiagaraFloat=50
     IsTarget:NiagaraFloat=1.6
     TargetPlatform:NiagaraDataInterfacePlatformSet=DataInterface() {propertyValues {Platforms {
        qualityLevelMask=-32
        deviceProfileStates [
         {profileName=IOS, qualityLevelMask=4294967295, setQualityLevelMask=4294967295}
         {profileName=Android, qualityLevelMask=4294967295, setQualityLevelMask=4294967295}
         {profileName="Android_Mali_G72 ", qualityLevelMask=4294967295, setQualityLevelMask=4294967295}
         {profileName=Android_Galaxy_Z_Fold2_Adreno, qualityLevelMask=4294967295, setQualityLevelMask=4294967295}
         {profileName=Android_Galaxy_Z_Fold2_Adreno_LRU, qualityLevelMask=4294967295, setQualityLevelMask=4294967295}
         {profileName=Android_Galaxy_Z_Fold2_Adreno_Vulkan, qualityLevelMask=4294967295, setQualityLevelMask=4294967295}
         {profileName=Android_High, qualityLevelMask=4294967295, setQualityLevelMask=4294967295}
         {profileName=IOS_Epic, qualityLevelMask=4294967295, setQualityLevelMask=4294967295}
        ]
       }}}
    )) {isSetParameters=true, module=SetVariables_#3}
  ]
  particleSpawnScript [
   &InitializeParticle(
    Color=<rgba 2 2 2 1>
    Lifetime=&RandomRangeFloat(Minimum=0.2, Maximum=0.5)
    Mesh Scale Mode=::NewEnumerator2
    UsePositionOffset=false
    Write Material Random=true
   )
   &InitialMeshOrientation(Rotation=&RandomRangeVector(Minimum=<0 0 0>, Maximum=<1 0 0>), Use Rotation Vector=true) {enabled=false}
   &InitialMeshOrientation(
    Rotation=&MakeVector(
     X=&RandomRangeFloat(Minimum=0, Maximum=1)
     Y=&RandomRangeFloat(Minimum=-0.05, Maximum=0.05)
     Z=&RandomRangeFloat(Minimum=-0.05, Maximum=0.05)
    )
    Use Orientation Vector=true
    Use Rotation Vector=true
   ) {module=InitialMeshOrientation001}
   &AddVelocityFromPoint(Module.Use Falloff Distance=false, Velocity Strength=&RandomRangeFloat(Minimum=350, Maximum=500))
  ]
  particleUpdateScript [
   &GravityForce() {enabled=false}
   &Drag(Drag=&RandomRangeFloat(Minimum=3, Maximum=4), Use Linear Drag=true, Use Rotational Drag=false)
   &MeshRotationForce(Rotation=&RandomRangeVector(Minimum=<5 0 0>, Maximum=<10 0 0>)) {enabled=false}
   &SolveForcesAndVelocity(Clamp Velocity=false, Limit Acceleration=false)
   &ParticleState()
   &ScaleMeshSize(Scale Factor=&VectorFromFloat(Value=&FloatFromCurve(
      FloatCurve=DataInterface() {propertyValues {Curve {
         defaultValue=0.294
         keys [
          {arriveTangent=4.1785, interpMode=RCIM_Cubic, leaveTangent=4.1785, tangentMode=RCTM_User, time=0, value=0}
          {arriveTangent=1.1179, interpMode=RCIM_Cubic, leaveTangent=1.1179, tangentMode=RCTM_User, time=0.2042, value=0.6228}
          {arriveTangent=0.0224, interpMode=RCIM_Cubic, leaveTangent=0.0224, tangentMode=RCTM_User, time=0.8, value=1}
          {arriveTangent=-8.6179, interpMode=RCIM_Cubic, leaveTangent=-8.6179, tangentMode=RCTM_User, time=1, value=0}
         ]
        }}}
      Scale Curve=$System.GlobalScale
     )))
  ]
  rendererClasses [&NiagaraMeshRendererProperties]
  renderers [{rendererClass=&NiagaraMeshRendererProperties, rendererIndex=0}]
 }
 Toon_Impact_Flare {
  emitterUpdateScript [
   &EmitterState()
   &SpawnBurst_Instantaneous(Spawn Count=1, Use Spawn Probability=false)
  ]
  particleSpawnScript [
   &InitializeParticle(
    Lifetime=0.2
    Mesh Scale Max=<1.45 1.45 1.45>
    Mesh Scale Min=<1.3 1.3 1.3>
    Mesh Scale Mode=::NewEnumerator2
    Sprite Size=&Multiply_Vector2D_ByFloat(Float=$System.GlobalScale, Vector2D=<200 175>)
    Sprite Size Mode=::NewEnumerator4
    UsePositionOffset=false
    Write Material Random=true
   ) {module=InitializeParticle001}
   /SetVariables(Particles.CoronaSpriteSize=&Vector2DFromFloat(Value=&Multiply_Float(A=90, B=$System.GlobalScale))) {isSetParameters=true}
   /SetVariables(Particles.ImpactColor:LinearColor=&MakeLinearColorFromVectorAndFloat("Vector (RGB)"=<1 1 1>, "Float (Alpha)"=&Multiply_Float(A=1, B=1))) {isSetParameters=true}
   &InitialMeshOrientation(
    Rotation=&UniformRangedVector(Minimum=<-1 -1 -1>, Maximum=<1 1 1>)
    Use Orientation Vector=true
    Use Rotation Vector=true
   )
  ]
  particleUpdateScript [
   &CameraOffset(Camera Offset Amount=30)
   &ParticleState()
   &DynamicMaterialParameters(
    Index 0 Param 1=&FloatFromCurve(FloatCurve=DataInterface() {propertyValues {Curve {keys [<at 0 1 RCIM_Cubic>,<at 0.8 0 RCIM_Cubic>]}}})
    Index 0 Param 2=2
    Index 0 Param 3=1
    Index 0 Param 4=1
    Index 1 Param 1=&FloatFromCurve(FloatCurve=DataInterface() {propertyValues {Curve {keys [
         {arriveTangent=-2.3163, interpMode=RCIM_Cubic, time=0, value=2}
         {arriveTangent=-1.9019, interpMode=RCIM_Cubic, leaveTangent=-1.9019, tangentMode=RCTM_User, time=1, value=0.4}
        ]}}})
    Index 1 Param 2=0.4
    Index 1 Param 3=0.6
    Index 1 Param 4=1
    Index 2 Param 1=&FloatFromCurve(FloatCurve=DataInterface() {propertyValues {Curve {keys [<at 0.46 0>,<at 1 1>]}}})
    Index 2 Param 2=45
    Module.Param0WriteEnabled=true
    Module.Param0WriteEnabled1=true
    Module.Param0WriteEnabled2=true
    Module.Param1WriteEnabled=true
    Module.Param1WriteEnabled1=true
    Module.Param1WriteEnabled2=true
    Module.Param2WriteEnabled=true
    Module.Param2WriteEnabled1=true
    Module.Param2WriteEnabled2=true
    Module.Param3WriteEnabled=true
    Module.Param3WriteEnabled1=true
    Module.Param3WriteEnabled2=true
    Write Parameter Index 1=true
    Write Parameter Index 2=true
   )
   /SetVariables(Particles.CoronaSpriteColor:LinearColor=&MakeLinearColorFromVectorAndFloat(
     "Vector (RGB)"=&MakeVectorFromLinearColorRGB(Color=<rgba 0.5 0.3315 0.189 1>)
     "Float (Alpha)"=&Multiply_Float(
      A=&FloatFromCurve(FloatCurve=DataInterface() {propertyValues {Curve {keys [<at 0.35 1.5 RCIM_Cubic>,<at 0.95 0>]}}})
      B=3.9
     )
    )) {isSetParameters=true}
   /SetVariables() {isSetParameters=true}
   &ScaleColor(
    Scale Alpha=&FloatFromCurve(FloatCurve=DataInterface() {propertyValues {Curve {keys [<at 0 1>,<at 1 0>]}}})
    Scale RGB=&MakeVectorFromLinearColorRGB(Color=<rgba 1 1 1 1>)
    ScaleA=true
    ScaleRGB=true
   )
   &ScaleMeshSize(Scale Factor=&VectorFromFloat(Value=&FloatFromCurve(FloatCurve=DataInterface() {propertyValues {Curve {keys [<at 0 0.294 RCIM_Cubic>,<at 1 1.05>]}}}, Scale Curve=$System.GlobalScale)))
  ]
  rendererClasses [&NiagaraSpriteRendererProperties, &NiagaraMeshRendererProperties]
  renderers [
   {rendererClass=&NiagaraSpriteRendererProperties}
   {rendererClass=&NiagaraMeshRendererProperties, rendererIndex=1}
  ]
 }
 Toon_Impact_Smoke {
  emitterUpdateScript [
   &EmitterState()
   &SpawnBurst_Instantaneous(Spawn Count=1, Spawn Time=0.03, Use Spawn Probability=false)
  ]
  particleSpawnScript [
   &InitializeParticle(
    Color=<rgba 0.8 0.6483 0.52 0.2>
    Lifetime Max=0.35
    Lifetime Min=0.3
    Lifetime Mode=::NewEnumerator1
    Sprite Rotation Angle=&RandomRangeFloat(Minimum=-45, Maximum=45)
    Sprite Rotation Mode=::NewEnumerator0
    Sprite Size Mode=::NewEnumerator3
    Uniform Sprite Size=&Multiply_Float(A=&RandomRangeFloat(Minimum=120, Maximum=140), B=$System.GlobalScale)
    UsePositionOffset=false
    Write Material Random=true
   ) {module=InitializeParticle001}
   &CameraOffset(Camera Offset Amount=30)
   &DynamicMaterialParameters(
    Index 0 Param 2=&UniformAOrBFloat(A=0)
    Module.Param0WriteEnabled=true
    Module.Param1WriteEnabled=true
    Module.Param2WriteEnabled=true
    Module.Param3WriteEnabled=true
   ) {module=DynamicMaterialParameters001}
  ]
  particleUpdateScript [
   &ParticleState()
   &SolveForcesAndVelocity(Clamp Velocity=false, Limit Acceleration=false)
   &DynamicMaterialParameters(
    Index 0 Param 1=&FloatFromCurve(FloatCurve=DataInterface() {propertyValues {Curve {keys [<at 0 0>,<at 1 1>]}}})
    Index 0 Param 2=1
    Module.Param0WriteEnabled=true
    Module.Param1WriteEnabled=true
    Module.Param2WriteEnabled=true
    Module.Param3WriteEnabled=true
   )
   &ScaleColor(
    Scale Alpha=&FloatFromCurve(FloatCurve=DataInterface() {propertyValues {Curve {keys [<at 0 1>,<at 0.8091 0.9932>,<at 1 0>]}}})
    ScaleA=true
    ScaleRGB=true
   )
  ]
  rendererClasses [&NiagaraSpriteRendererProperties]
  renderers [{rendererClass=&NiagaraSpriteRendererProperties}]
 }
 Toon_Ring {
  emitterUpdateScript [
   &EmitterState(Life Cycle Mode=::NewEnumerator1, Loop Behavior=::NewEnumerator1, Loop Delay=0.05, Loop Duration=0.15, UseLoopDelay=true)
   &SpawnBurst_Instantaneous(Spawn Count=1, Use Spawn Probability=false)
   /SetVariables(Emitter.FresnelThickEXP:NiagaraFloat=&self:PlatFromSelect(
     Default:NiagaraFloat=50
     IsTarget:NiagaraFloat=1.6
     TargetPlatform:NiagaraDataInterfacePlatformSet=DataInterface() {propertyValues {Platforms {
        qualityLevelMask=-32
        deviceProfileStates [
         {profileName=IOS, qualityLevelMask=4294967295, setQualityLevelMask=4294967295}
         {profileName=Android, qualityLevelMask=4294967295, setQualityLevelMask=4294967295}
         {profileName="Android_Mali_G72 ", qualityLevelMask=4294967295, setQualityLevelMask=4294967295}
         {profileName=Android_Galaxy_Z_Fold2_Adreno, qualityLevelMask=4294967295, setQualityLevelMask=4294967295}
         {profileName=Android_Galaxy_Z_Fold2_Adreno_LRU, qualityLevelMask=4294967295, setQualityLevelMask=4294967295}
         {profileName=Android_Galaxy_Z_Fold2_Adreno_Vulkan, qualityLevelMask=4294967295, setQualityLevelMask=4294967295}
         {profileName=Android_High, qualityLevelMask=4294967295, setQualityLevelMask=4294967295}
         {profileName=IOS_Epic, qualityLevelMask=4294967295, setQualityLevelMask=4294967295}
        ]
       }}}
    )) {isSetParameters=true, module=SetVariables_#3}
  ]
  particleSpawnScript [
   &InitializeParticle(Lifetime=0.35, Mesh Scale Mode=::NewEnumerator3, Mesh Uniform Scale=2.25, UsePositionOffset=false, Write Material Random=true) {module=InitializeParticle001}
   &InitialMeshOrientation(
    Rotation=&MakeVector(
     X=&RandomRangeFloat(Minimum=0, Maximum=1)
     Y=&RandomRangeFloat(Minimum=-0.05, Maximum=0.05)
     Z=&RandomRangeFloat(Minimum=-0.05, Maximum=0.05)
    )
    Use Orientation Vector=true
    Use Rotation Vector=true
   )
  ]
  particleUpdateScript [
   &CameraOffset(Camera Offset Amount=30)
   &ParticleState()
   &ScaleMeshSize(Scale Factor=&VectorFromFloat(Value=&FloatFromCurve(
      FloatCurve=DataInterface() {propertyValues {Curve {
         defaultValue=0.294
         keys [
          {arriveTangent=4.1785, interpMode=RCIM_Cubic, leaveTangent=4.1785, tangentMode=RCTM_User, time=0, value=0}
          {arriveTangent=2.2121, interpMode=RCIM_Cubic, leaveTangent=2.2121, tangentMode=RCTM_User, time=0.2, value=0.7}
          {arriveTangent=0.1291, interpMode=RCIM_Cubic, leaveTangent=-0.5945, tangentMode=RCTM_Break, time=0.6, value=1}
          {arriveTangent=-5.7296, interpMode=RCIM_Cubic, leaveTangent=-5.7296, tangentMode=RCTM_User, time=1, value=0}
         ]
        }}}
      Scale Curve=$System.GlobalScale
     )))
   &ScaleMeshSize(Scale Factor=&VectorFromCurve(VectorCurve=DataInterface() {propertyValues {
       XCurve {keys [
         {interpMode=RCIM_Cubic, leaveTangentWeight=0.3333, tangentMode=RCTM_User, time=0, value=1}
         {
          arriveTangent=-0.051
          arriveTangentWeight=0.1167
          interpMode=RCIM_Cubic
          leaveTangent=-0.051
          leaveTangentWeight=0.2167
          tangentMode=RCTM_User
          time=0.3376
          value=0.9587
         }
         {
          arriveTangent=0.2435
          arriveTangentWeight=0.1
          interpMode=RCIM_Cubic
          leaveTangent=0.2435
          leaveTangentWeight=0.1167
          tangentMode=RCTM_User
          time=0.6489
          value=0.9804
         }
         {arriveTangent=-2.7922, interpMode=RCIM_Cubic, leaveTangent=-2.7922, time=0.85, value=0.8}
         <at 1 0>
        ]}
       YCurve {keys [
         {arriveTangent=-0.6287, interpMode=RCIM_Cubic, leaveTangent=-0.6287, leaveTangentWeight=0.2833, tangentMode=RCTM_User, time=0, value=1.0022}
         {
          arriveTangent=-0.3763
          arriveTangentWeight=0.0833
          interpMode=RCIM_Cubic
          leaveTangent=-0.3763
          leaveTangentWeight=0.2
          tangentMode=RCTM_User
          time=0.25
          value=0.9829
         }
         {arriveTangent=-0.0441, interpMode=RCIM_Cubic, leaveTangent=-0.0441, time=0.4966, value=0.9755}
         {arriveTangent=0.014, interpMode=RCIM_Cubic, leaveTangent=0.014, tangentMode=RCTM_User, time=0.6511, value=0.9652}
         {arriveTangent=-2.7667, interpMode=RCIM_Cubic, leaveTangent=-2.7667, time=0.85, value=0.8}
         <at 1 0>
        ]}
       ZCurve {keys [<at 0 1>,<at 1 1>]}
      }})) {enabled=false, module=ScaleMeshSize001}
  ]
  rendererClasses [&NiagaraMeshRendererProperties]
  renderers [
   {rendererClass=&NiagaraMeshRendererProperties, rendererIndex=0}
   {rendererClass=&NiagaraMeshRendererProperties, rendererIndex=1}
  ]
 }
 Toon_SpeedLines {
  emitterUpdateScript [
   &EmitterState(Life Cycle Mode=::NewEnumerator1, Loop Behavior=::NewEnumerator1, Loop Duration=1, UseLoopDelay=false)
   &SpawnBurst_Instantaneous(Spawn Count=1, Use Spawn Probability=false) {module=SpawnBurst_Instantaneous001}
   &SpawnBurst_Instantaneous(Spawn Time=0.0459, Use Spawn Probability=false, UseLoopCountLimit=false)
  ]
  particleSpawnScript [&InitializeParticle(
    Lifetime=0.3
    Mesh Scale Mode=::NewEnumerator3
    Mesh Uniform Scale=&Multiply_Float(A=2.5, B=&RandomRangeFloat(Minimum=1, Maximum=1.8))
    Sprite Rotation Mode=::NewEnumerator2
    Sprite Size Mode=::NewEnumerator3
    Uniform Sprite Size=&Multiply_Float(A=165, B=&RandomRangeFloat(Minimum=0.7, Maximum=1))
    UsePositionOffset=false
    Write Material Random=true
   ) {module=InitializeParticle001}]
  particleUpdateScript [
   &ParticleState()
   &SolveForcesAndVelocity(Clamp Velocity=false, Limit Acceleration=false)
   &ScaleSpriteSize(Scale Factor=&Vector2DFromFloat(Value=&FloatFromCurve(
      FloatCurve=DataInterface() {propertyValues {Curve {keys [{arriveTangent=3.9086, interpMode=RCIM_Cubic, leaveTangent=3.9086, tangentMode=RCTM_User, time=0, value=0}, <at 1 1.4 RCIM_Cubic>]}}}
      Scale Curve=$System.GlobalScale
     )))
   &ScaleMeshSize(Scale Factor=&VectorFromFloat(Value=&FloatFromCurve(
      FloatCurve=DataInterface() {propertyValues {Curve {keys [{arriveTangent=2.9993, interpMode=RCIM_Cubic, leaveTangent=2.9993, tangentMode=RCTM_User, time=0, value=0}, <at 1 1 RCIM_Cubic>]}}}
      Scale Curve=$System.GlobalScale
     )))
   &DynamicMaterialParameters(
    Index 0 Param 1=&FloatFromCurve(FloatCurve=DataInterface() {propertyValues {Curve {keys [<at 0 1 RCIM_Cubic>, {arriveTangent=-2.742, interpMode=RCIM_Cubic, leaveTangent=-2.742, tangentMode=RCTM_User, time=1, value=0}]}}})
    Index 0 Param 2=&FloatFromCurve(FloatCurve=DataInterface() {propertyValues {Curve {keys [<at 0 0 RCIM_Cubic>, {arriveTangent=2.4833, interpMode=RCIM_Cubic, leaveTangent=2.4833, tangentMode=RCTM_User, time=1, value=1}]}}})
    Module.Param0WriteEnabled=true
    Module.Param1WriteEnabled=true
    Module.Param2WriteEnabled=true
    Module.Param3WriteEnabled=true
   )
   &ScaleColor(
    Scale Alpha=&FloatFromCurve(FloatCurve=DataInterface() {propertyValues {Curve {keys [
         {arriveTangent=-0.2921, interpMode=RCIM_Cubic, leaveTangent=-0.2921, tangentMode=RCTM_User, time=0, value=7}
         {arriveTangent=-7, interpMode=RCIM_Cubic, leaveTangent=-7, time=0.4, value=4}
         <at 1 0 RCIM_Cubic>
        ]}}})
    ScaleA=true
    ScaleRGB=true
   )
   &CameraOffset(Camera Offset Amount=10)
  ]
  rendererClasses [&NiagaraSpriteRendererProperties, &NiagaraMeshRendererProperties]
  renderers [
   {rendererClass=&NiagaraSpriteRendererProperties}
   {rendererClass=&NiagaraMeshRendererProperties, rendererIndex=1}
  ]
 }
 Toon_Spikes {
  emitterUpdateScript [
   &EmitterState(Life Cycle Mode=::NewEnumerator1, Loop Behavior=::NewEnumerator1, Loop Duration=1, UseLoopDelay=false)
   &SpawnRate(SpawnRate=20, Use Spawn Probability=false) {enabled=false}
   &SpawnBurst_Instantaneous(Spawn Count=8, Use Spawn Probability=false)
  ]
  particleSpawnScript [
   &InitializeParticle(
    Color Maximum=<rgba 0.025 0.0165 0.0095 1>
    Color Minimum=<rgba 0.025 0.0165 0.0095 0.7>
    Color Mode=::NewEnumerator2
    Lifetime Max=0.2
    Lifetime Min=0.1
    Lifetime Mode=::NewEnumerator1
    Sprite Rotation Mode=::NewEnumerator2
    Sprite Size Max=<10 45>
    Sprite Size Min=<5 33.5>
    Sprite Size Mode=::NewEnumerator2
    UsePositionOffset=false
    Write Material Random=true
   ) {module=InitializeParticle001}
   &AddVelocityFromPoint(Module.Use Falloff Distance=false, Velocity Strength=&Multiply_Float(A=140, B=$System.GlobalScale))
  ]
  particleUpdateScript [
   &CameraOffset(Camera Offset Amount=10)
   &ParticleState()
   &SolveForcesAndVelocity(Clamp Velocity=false, Limit Acceleration=false)
   &ScaleSpriteSize(Scale Factor=&Vector2DFromCurve(
     Vector2Curve=DataInterface() {propertyValues {XCurve {keys [<at 0 0.778>,<at 1 1>]}, YCurve {keys [<at 0 0.03>,<at 1 1.5>]}}}
     ScaleCurve=&Multiply_Vector2D_ByFloat(Float=$System.GlobalScale, Vector2D=<1 1>)
    ))
   &ScaleColor(
    Scale RGB=&VectorFromFloat(Value=&FloatFromCurve(FloatCurve=DataInterface() {propertyValues {Curve {keys [<at 0 10>,<at 1 0.75>]}}}))
    ScaleA=true
    ScaleRGB=true
   )
  ]
  rendererClasses [&NiagaraSpriteRendererProperties]
  renderers [{rendererClass=&NiagaraSpriteRendererProperties}]
 }
 Toon_Star {
  emitterUpdateScript [
   &EmitterState(Life Cycle Mode=::NewEnumerator1, Loop Behavior=::NewEnumerator1, Loop Delay=0.05, Loop Duration=0.15, UseLoopDelay=true)
   &SpawnBurst_Instantaneous(Spawn Count=1, Use Spawn Probability=false)
   /SetVariables(Emitter.FresnelThickEXP:NiagaraFloat=&self:PlatFromSelect(
     Default:NiagaraFloat=50
     IsTarget:NiagaraFloat=1.6
     TargetPlatform:NiagaraDataInterfacePlatformSet=DataInterface() {propertyValues {Platforms {
        qualityLevelMask=-32
        deviceProfileStates [
         {profileName=IOS, qualityLevelMask=4294967295, setQualityLevelMask=4294967295}
         {profileName=Android, qualityLevelMask=4294967295, setQualityLevelMask=4294967295}
         {profileName="Android_Mali_G72 ", qualityLevelMask=4294967295, setQualityLevelMask=4294967295}
         {profileName=Android_Galaxy_Z_Fold2_Adreno, qualityLevelMask=4294967295, setQualityLevelMask=4294967295}
         {profileName=Android_Galaxy_Z_Fold2_Adreno_LRU, qualityLevelMask=4294967295, setQualityLevelMask=4294967295}
         {profileName=Android_Galaxy_Z_Fold2_Adreno_Vulkan, qualityLevelMask=4294967295, setQualityLevelMask=4294967295}
         {profileName=Android_High, qualityLevelMask=4294967295, setQualityLevelMask=4294967295}
         {profileName=IOS_Epic, qualityLevelMask=4294967295, setQualityLevelMask=4294967295}
        ]
       }}}
    )) {isSetParameters=true}
  ]
  particleSpawnScript [
   &InitializeParticle(
    Color=<rgba 5 5 5 1>
    Lifetime=0.3
    Mesh Scale Mode=::NewEnumerator3
    Mesh Uniform Scale=1.75
    UsePositionOffset=false
    Write Material Random=true
   ) {module=InitializeParticle001}
   &InitialMeshOrientation(
    Rotation=&MakeVector(
     X=&RandomRangeFloat(Minimum=-0.1, Maximum=0.1)
     Y=&RandomRangeFloat(Minimum=-0.1, Maximum=0.1)
     Z=&MakeCustomFloatFromBool(Bool=&RandomBool(), True Float=0.025, False Float=-0.025)
    )
    Use Orientation Vector=true
    Use Rotation Vector=true
   )
   /SetVariables(
    Particles.Flash Flare Color:LinearColor=$System.TOD Color
    Particles.Flash Flare Lifetime=0.3
    Particles.Flash Flare Sprite Size:Vector2f=&Multiply_Vector2D_ByFloat(Float=&FN_FrontEndFloatModifier(Gameplay=1.5, Front End=$User.FrontEndScale), Vector2D=<180 180>)
   ) {isSetParameters=true}
   &DynamicMaterialParameters(
    Index 0 Param 1=0.2
    Index 0 Param 2=1
    Module.Index0Param0Write=true
    Module.Index0Param1Write=true
    Module.Index0Param2Write=true
    Module.Index0Param3Write=true
   )
   &CameraOffset(Camera Offset Amount=30)
  ]
  particleUpdateScript [
   &ParticleState()
   &ScaleMeshSize(Scale Factor=&VectorFromFloat(Value=&FloatFromCurve(
      FloatCurve=DataInterface() {propertyValues {Curve {
         defaultValue=0.294
         keys [
          {arriveTangent=4.1785, interpMode=RCIM_Cubic, leaveTangent=4.1785, tangentMode=RCTM_User, time=0, value=0}
          {arriveTangent=1.1179, interpMode=RCIM_Cubic, leaveTangent=1.1179, tangentMode=RCTM_User, time=0.2042, value=0.6228}
          {arriveTangent=0.0224, interpMode=RCIM_Cubic, leaveTangent=0.0224, tangentMode=RCTM_User, time=0.8, value=1}
          {arriveTangent=-8.6179, interpMode=RCIM_Cubic, leaveTangent=-8.6179, tangentMode=RCTM_User, time=1, value=0}
         ]
        }}}
      Scale Curve=$System.GlobalScale
     )))
   &ScaleMeshSize(Scale Factor=&VectorFromCurve(VectorCurve=DataInterface() {propertyValues {
       XCurve {keys [
         {interpMode=RCIM_Cubic, leaveTangentWeight=0.3333, tangentMode=RCTM_User, time=0, value=1}
         {
          arriveTangent=0.3114
          arriveTangentWeight=0.1167
          interpMode=RCIM_Cubic
          leaveTangent=0.3114
          leaveTangentWeight=0.2167
          tangentMode=RCTM_User
          time=0.3376
          value=0.9587
         }
         {
          arriveTangent=0.2435
          arriveTangentWeight=0.1
          interpMode=RCIM_Cubic
          leaveTangent=0.2435
          leaveTangentWeight=0.1167
          tangentMode=RCTM_User
          time=0.6489
          value=0.9804
         }
         {arriveTangent=-2.7922, interpMode=RCIM_Cubic, leaveTangent=-2.7922, time=0.85, value=0.8}
         <at 1 0>
        ]}
       YCurve {keys [
         {arriveTangent=-0.6287, interpMode=RCIM_Cubic, leaveTangent=-0.6287, leaveTangentWeight=0.2833, tangentMode=RCTM_User, time=0, value=1.025}
         {
          arriveTangent=-0.3763
          arriveTangentWeight=0.0833
          interpMode=RCIM_Cubic
          leaveTangent=-0.3763
          leaveTangentWeight=0.2
          tangentMode=RCTM_User
          time=0.25
          value=1
         }
         {arriveTangent=-0.1766, interpMode=RCIM_Cubic, leaveTangent=-0.1766, time=0.4966, value=0.9755}
         {arriveTangent=-0.4967, interpMode=RCIM_Cubic, leaveTangent=-0.4967, time=0.65, value=0.9294}
         {arriveTangent=-2.6553, interpMode=RCIM_Cubic, leaveTangent=-2.6553, time=0.85, value=0.8}
         <at 1 0>
        ]}
       ZCurve {keys [<at 0 1>,<at 1 1>]}
      }})) {module=ScaleMeshSize001}
   /SetVariables(Particles.Flash Flare Age:NiagaraFloat=&Divide_Float(A=$Particles.Age, B=$Particles.Initial.Flash Flare Lifetime)) {isSetParameters=true}
   /SetVariables(Particles.Flash Flare Color:LinearColor=&MakeLinearColorFromVectorAndFloat(
     "Vector (RGB)"=&Multiply_Vector(
      A=&Multiply_Vector(
       A=&MakeVectorFromLinearColorRGB(Color=$Particles.Initial.Flash Flare Color)
       B=&MakeVectorFromLinearColorRGB(Color=&ColorFromCurve(
         ColorCurve=DataInterface() {propertyValues {
           AlphaCurve {keys [<at 0 1 RCIM_Cubic>,<at 1 0 RCIM_Cubic>]}
           BlueCurve {keys [<at 0.0198 3>,<at 0.491 0.5833>,<at 1 0.1667>]}
           GreenCurve {keys [<at 0.0198 0>,<at 0.491 0>,<at 1 0>]}
           RedCurve {keys [<at 0.0198 3>,<at 0.491 1>,<at 1 1>]}
          }}
         CurveIndex=$Particles.Flash Flare Age
        ))
      )
      B=&VectorFromFloat(Value=&FN_FrontEndFloatModifier(Gameplay=0.75, Front End=1))
     )
     "Float (Alpha)"=&FloatFromCurve(
      FloatCurve=DataInterface() {propertyValues {Curve {
         defaultValue=0.294
         keys [
          {arriveTangent=4.1785, interpMode=RCIM_Cubic, leaveTangent=4.1785, tangentMode=RCTM_User, time=0, value=0}
          {arriveTangent=1.1179, interpMode=RCIM_Cubic, leaveTangent=1.1179, tangentMode=RCTM_User, time=0.2042, value=0.6228}
          {arriveTangent=0.0224, interpMode=RCIM_Cubic, leaveTangent=0.0224, tangentMode=RCTM_User, time=0.8, value=1}
          {arriveTangent=-8.6179, interpMode=RCIM_Cubic, leaveTangent=-8.6179, tangentMode=RCTM_User, time=1, value=0}
         ]
        }}}
      CurveIndex=$Particles.Flash Flare Age
      Scale Curve=0.75
     )
    )) {isSetParameters=true}
  ]
  rendererClasses [&NiagaraMeshRendererProperties, &NiagaraSpriteRendererProperties]
  renderers [
   {rendererClass=&NiagaraMeshRendererProperties, rendererIndex=0}
   {rendererClass=&NiagaraMeshRendererProperties, rendererIndex=1}
   {rendererClass=&NiagaraSpriteRendererProperties, rendererIndex=2}
  ]
 }
}
summary {
 systemName=NS_Pickaxe_GiraffeScallion_Impact
 emitters [
  {emitterName=Toon_Impact_Smoke, rendererClasses [&NiagaraSpriteRendererProperties]}
  {
   emitterName=Toon_SpeedLines
   rendererClasses [&NiagaraSpriteRendererProperties, &NiagaraMeshRendererProperties]
  }
  {
   emitterName=Toon_Impact_Flare
   rendererClasses [&NiagaraSpriteRendererProperties, &NiagaraMeshRendererProperties]
  }
  {emitterName=Toon_Spikes, rendererClasses [&NiagaraSpriteRendererProperties]}
  {emitterName=Toon_Hearts, rendererClasses [&NiagaraMeshRendererProperties]}
  {emitterName=Toon_Star, rendererClasses [&NiagaraMeshRendererProperties, &NiagaraSpriteRendererProperties]}
  {emitterName=Toon_Ring, rendererClasses [&NiagaraMeshRendererProperties]}
 ]
 userVariables [
  User.bIsFrontEnd {type=NiagaraBool}
  User.bIsFrontEndPreview {type=NiagaraBool}
  User.FrontEndScale {type=NiagaraFloat, defaultValue=0, description=""}
  User.FrontendScaler {type=NiagaraFloat, defaultValue=0.55, description=""}
  User.PhysicalSurface {type=NiagaraInt32}
 ]
}
