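# Niagara swing effect for the GiraffeScallion pickaxe, drawing two short-lived ribbon trails during melee swings.
# MainTrail and a pink RingTrail use shared anim-trail modules to spawn and update ribbons from bone or socket pairs.
# User.SwingSectionIndex selects a system-age curve that cuts spawn alpha to match each swing section.
# Color-over-life and alpha curves shape each trail, with time-of-day tinting applied to the color.
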NiagaraSystem /BRCosmetics/Weapons/FORT_Melee/Pickaxe_GiraffeScallion/FX/NS_Pickaxe_GiraffeScallion_Swing
dependencies {
 usedDataInterfaces [&NiagaraDataInterfaceCurve, &NiagaraDataInterfaceColorCurve, &NiagaraDataInterfaceSocketReader]
 usedDynamicInputs [&Add_Float]
 usedModules [
  &self:SystemSpawnScript.SystemScriptSource.SystemScriptGraph.NiagaraNodeAssignment_4.SetVariables_#1
  &SystemState
  &self:SystemSpawnScript.SystemScriptSource.SystemScriptGraph.NiagaraNodeAssignment_2.SetVariables_#2
 ]
}
emitters {
 MainTrail {
  emitterUpdateScript [
   &EmitterState()
   &FN_SpawnAnimTrail()
  ]
  particleSpawnScript [
   &FN_InitializeAnimTrailBoneIndecies()
   &FN_InitializeAnimTrail(
    Color=&MultiplyLinearColor(
     Linear Color=<rgba 1 1 1 0.5>
     Scale Factor=&MakeFloat4FromVectorAndFloat(
      "Vector (RGB)"=&VectorFromFloat(Value=1)
      "Float (Alpha)"=&Lerp_Float(
       A=&MergeHelper_FloatSwitch(
        0=&FloatFromCurve(FloatCurve=DataInterface() {propertyValues {Curve {keys [<at 0.1 1>,<at 0.12 0>]}}}, CurveIndex=$Engine.System.Age)
        1=&FloatFromCurve(FloatCurve=DataInterface() {propertyValues {Curve {keys [<at 0.1 1>,<at 0.11 0>]}}}, CurveIndex=$Engine.System.Age)
        2=&FloatFromCurve(FloatCurve=DataInterface() {propertyValues {Curve {keys [<at 0.13 1>,<at 0.14 0>]}}}, CurveIndex=$Engine.System.Age)
        3=&FloatFromCurve(FloatCurve=DataInterface() {propertyValues {Curve {keys [<at 0.13 1>,<at 0.2 0>]}}}, CurveIndex=$Engine.System.Age)
        Index=$User.SwingSectionIndex
        Num=::NewEnumerator2
       )
       B=$System.TrailFade
       Alpha=0
      )
     )
    )
    Lifetime=0.2
   )
   &AddVelocity(Velocity=<0 400 0>) {enabled=false}
  ]
  particleUpdateScript [
   &SolveForcesAndVelocity(Clamp Velocity=false, Limit Acceleration=false) {enabled=false}
   &FN_UpdateAnimTrail() {module=FN_UpdateAnimTrail001}
   &ScaleColor(
    Scale Alpha=&FloatFromCurve(FloatCurve=DataInterface() {propertyValues {Curve {
        defaultValue=1
        keys [
         {tangentMode=RCTM_User, time=0, value=1}
         {arriveTangent=-0.0405, interpMode=RCIM_Cubic, leaveTangent=-0.0405, tangentMode=RCTM_User, time=0.5, value=1}
         {arriveTangent=-8.0089, interpMode=RCIM_Cubic, leaveTangent=-8.0089, tangentMode=RCTM_User, time=0.8, value=0}
        ]
       }}})
    Scale RGB=&Multiply_Vector(
     A=&MakeVectorFromLinearColorRGB(Color=&ColorFromCurve(ColorCurve=DataInterface() {propertyValues {
         AlphaCurve {keys [<at 0 1 RCIM_Cubic>]}
         BlueCurve {keys [{arriveTangent=0.96, interpMode=RCIM_Cubic, time=0.1316, value=1}, {interpMode=RCIM_Cubic, leaveTangent=-0.1644, time=1, value=1}]}
         GreenCurve {keys [{arriveTangent=0.0419, interpMode=RCIM_Cubic, time=0.1316, value=1}, {interpMode=RCIM_Cubic, leaveTangent=0.2347, time=1, value=1}]}
         RedCurve {keys [{arriveTangent=-0.4464, interpMode=RCIM_Cubic, time=0.1316, value=1}, <at 1 1 RCIM_Cubic>]}
        }}))
     B=&MakeVectorFromLinearColorRGB(Color=&FNTimeOfDayColor(Night Color=<rgba 0.21 0.3325 0.7 1>, Morning Color=<rgba 1 1 1 1>, Day Color=<rgba 1 1 1 1>, Evening Color=<rgba 1 1 1 1>))
    )
    ScaleA=true
    ScaleRGB=true
   ) {module=ScaleColor001}
  ]
  rendererClasses [&NiagaraRibbonRendererProperties]
  renderers [{rendererClass=&NiagaraRibbonRendererProperties}]
 }
 RingTrail {
  emitterUpdateScript [
   &EmitterState()
   &FN_SpawnAnimTrail()
  ]
  particleSpawnScript [
   &FN_InitializeAnimTrailBoneIndecies(First Bone or Socket Index=4, Second Bone or Socket Index=5)
   &FN_InitializeAnimTrail(
    AnimTrailWidth=0.9
    Color=&MultiplyLinearColor(
     Linear Color=<rgba 0.8174 0.0817 0.6575 0.5>
     Scale Factor=&MakeFloat4FromVectorAndFloat(
      "Vector (RGB)"=&MakeVectorFromLinearColorRGB(Color=$System.TOD)
      "Float (Alpha)"=&Lerp_Float(
       A=&MergeHelper_FloatSwitch(
        0=&FloatFromCurve(FloatCurve=DataInterface() {propertyValues {Curve {keys [<at 0.1 1>,<at 0.12 0>]}}}, CurveIndex=$Engine.System.Age)
        1=&FloatFromCurve(FloatCurve=DataInterface() {propertyValues {Curve {keys [<at 0.1 1>,<at 0.11 0>]}}}, CurveIndex=$Engine.System.Age)
        2=&FloatFromCurve(FloatCurve=DataInterface() {propertyValues {Curve {keys [<at 0.13 1>,<at 0.14 0>]}}}, CurveIndex=$Engine.System.Age)
        3=&FloatFromCurve(FloatCurve=DataInterface() {propertyValues {Curve {keys [<at 0.13 1>,<at 0.2 0>]}}}, CurveIndex=$Engine.System.Age)
        Index=$User.SwingSectionIndex
        Num=::NewEnumerator2
       )
       B=$System.TrailFade
       Alpha=0
      )
     )
    )
    Lifetime=0.2
   )
   &AddVelocity(Velocity=<0 400 0>) {enabled=false}
  ]
  particleUpdateScript [
   &SolveForcesAndVelocity(Clamp Velocity=false, Limit Acceleration=false) {enabled=false}
   &FN_UpdateAnimTrail() {module=FN_UpdateAnimTrail001}
   &ScaleColor(
    Scale Alpha=&FloatFromCurve(
     FloatCurve=DataInterface() {propertyValues {Curve {
        defaultValue=1
        keys [
         {tangentMode=RCTM_User, time=0, value=1}
         {arriveTangent=-0.0405, interpMode=RCIM_Cubic, leaveTangent=-0.0405, tangentMode=RCTM_User, time=0.7, value=1}
         {arriveTangent=-8.0089, interpMode=RCIM_Cubic, leaveTangent=-8.0089, tangentMode=RCTM_User, time=1, value=0}
        ]
       }}}
     Scale Curve=0.25
    )
    Scale RGB=&Multiply_Vector(
     A=&MakeVectorFromLinearColorRGB(Color=&ColorFromCurve(ColorCurve=DataInterface() {propertyValues {
         AlphaCurve {keys [<at 0 1 RCIM_Cubic>]}
         BlueCurve {keys [{arriveTangent=0.96, interpMode=RCIM_Cubic, time=0.1316, value=1}, {interpMode=RCIM_Cubic, leaveTangent=-0.1644, time=1, value=1}]}
         GreenCurve {keys [{arriveTangent=0.0419, interpMode=RCIM_Cubic, time=0.1316, value=1}, {interpMode=RCIM_Cubic, leaveTangent=0.2347, time=1, value=1}]}
         RedCurve {keys [{arriveTangent=-0.4464, interpMode=RCIM_Cubic, time=0.1316, value=1}, <at 1 1 RCIM_Cubic>]}
        }}))
     B=&MakeVectorFromLinearColorRGB(Color=&FNTimeOfDayColor(Night Color=<rgba 1 1 1 1>, Morning Color=<rgba 1 1 1 1>, Day Color=<rgba 1 1 1 1>, Evening Color=<rgba 1 1 1 1>))
    )
    ScaleA=true
    ScaleRGB=true
   ) {module=ScaleColor001}
  ]
  rendererClasses [&NiagaraRibbonRendererProperties]
  renderers [{rendererClass=&NiagaraRibbonRendererProperties}]
 }
}
summary {
 systemName=NS_Pickaxe_GiraffeScallion_Swing
 emitters [
  {emitterName=MainTrail, rendererClasses [&NiagaraRibbonRendererProperties]}
  {emitterName=RingTrail, rendererClasses [&NiagaraRibbonRendererProperties]}
 ]
 userVariables [
  User.RibbonIDOffset {type=NiagaraInt32}
  User.SwingSectionIndex {type=NiagaraInt32}
 ]
}
