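# Animation Blueprint for the GiraffeScallion pickaxe that adds procedural secondary motion to its dangling wire, microphone and star bones.
# The input pose runs in component space through three chained Anim Dynamics nodes with angular springs, damping and gravity overrides, then returns to local space.
# The wire adds a linear spring, and the star simulates bone-relative while the others are root-relative.
# A second star node is unconnected, and the event graph holds only unwired default nodes.
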AnimBlueprint /BRCosmetics/Weapons/FORT_Melee/Pickaxe_GiraffeScallion/Meshes/GiraffeScallion_Axe_AnimBP
target_skeleton=&/GiraffeScallion_Axe_Skeleton
graphs [
 AnimGraph {nodes [
   Anim Dynamics - Bone: dyn_star_01(Alpha=1, ExternalForce="0,0,0") {
    angularDampingOverride=1
    angularSpringConstant=500
    bAngularSpring=true
    bOverrideAngularDamping=true
    bUseGravityOverride=true
    simulationSpace=RootRelative
    alphaBoolBlend {blendOption=Linear}
    boundBone {boneName=dyn_star_01}
    componentAppliedLinearAccClamp=<1000 1000 1000>
    componentLinearAccScale=<0.5 0.5 0.5>
    componentLinearVelScale=<0.5 0.5 0.5>
    gravityOverride {z=-980}
    physicsBodyDefinitions [{
      boundBone {boneName=dyn_star_01}
      boxExtents=<30 30 30>
      constraintSetup {angularLimitsMax=<10 35 0>, angularLimitsMin=<-10 0 -10>, angularTarget=<0 0 0>}
      localJointOffset=<-1 0 0>
     }]
   }
   AnimGraphNode_SubInput_0 = Input Pose()
   n1 = Anim Dynamics - Bone: dyn_wire_01(ComponentPose=Local To Component(LocalPose=@AnimGraphNode_SubInput_0), Alpha=1, ExternalForce="0,0,0") {
    angularDampingOverride=0.999
    angularSpringConstant=8000
    bAngularSpring=true
    bLinearSpring=true
    bOverrideAngularDamping=true
    bOverrideLinearDamping=true
    bUseGravityOverride=true
    linearDampingOverride=0.9999
    linearSpringConstant=250
    simulationSpace=RootRelative
    alphaBoolBlend {blendOption=Linear}
    boundBone {boneName=dyn_wire_01}
    componentAppliedLinearAccClamp=<1000 1000 1200>
    componentLinearAccScale=<0.5 0.5 0.5>
    componentLinearVelScale=<0.5 0.5 0.5>
    gravityOverride {z=-500}
    physicsBodyDefinitions [{
      boundBone {boneName=dyn_wire_01}
      boxExtents=<25 25 25>
      constraintSetup {angularLimitsMax=<25 0 0>, angularLimitsMin=<-25 0 0>, angularTarget=<0 0 0>, linearAxesMax=<1 0 0>, linearAxesMin=<-1 0 0>}
      localJointOffset=<-1 0 0>
     }]
   }
   n2 = Anim Dynamics - Bone: dyn_microphon_01(ComponentPose=@n1, Alpha=1, ExternalForce="0,0,0") {
    angularDampingOverride=0.99
    angularSpringConstant=500
    bAngularSpring=true
    bOverrideAngularDamping=true
    bUseGravityOverride=true
    simulationSpace=RootRelative
    alphaBoolBlend {blendOption=Linear}
    boundBone {boneName=dyn_microphon_01}
    componentAppliedLinearAccClamp=<1000 1000 1200>
    componentLinearAccScale=<0.5 0.5 0.5>
    componentLinearVelScale=<0.5 0.5 0.5>
    gravityOverride {z=-700}
    physicsBodyDefinitions [{
      boundBone {boneName=dyn_microphon_01}
      boxExtents=<40 40 40>
      constraintSetup {angularLimitsMax=<0 0 0>, angularLimitsMin=<0 0 -30>, angularTarget=<0 0 0>}
      localJointOffset=<-1 0 0>
     }]
   }
   n3 = Anim Dynamics - Bone: dyn_star_01(ComponentPose=@n2, Alpha=1, ExternalForce="0,0,0") {
    angularDampingOverride=0.99
    angularSpringConstant=5000
    bAngularSpring=true
    bOverrideAngularDamping=true
    bUseGravityOverride=true
    simulationSpace=BoneRelative
    alphaBoolBlend {blendOption=Linear}
    boundBone {boneName=dyn_star_01}
    componentAppliedLinearAccClamp=<1000 1000 1000>
    componentLinearAccScale=<0.5 0.5 0.5>
    componentLinearVelScale=<0.5 0.5 0.5>
    gravityOverride {z=-980}
    physicsBodyDefinitions [{
      boundBone {boneName=dyn_star_01}
      boxExtents=<15 15 15>
      constraintSetup {angularLimitsMax=<10 20 0>, angularLimitsMin=<-10 0 -10>, angularTarget=<0 0 0>}
      localJointOffset=<-1 0 0>
     }]
   }
   Output Pose(Result=Component To Local(ComponentPose=@n3)) {layerGroup=None}
  ]}
 EventGraph {nodes [
   TryGetPawnOwner()
   Event BlueprintUpdateAnimation()
  ]}
]
variables []
